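# Physics asset for the female Neon Graffiti cosmetic body that defines secondary motion for its hood and two drawstrings.
# Capsules on spine_04 and spine_05 form the torso volume, with small anchor spheres for the hood and each string joined to spine_05 by angularly free constraints.
# Each string is a mirrored three-box chain with limited swing and almost no twist, while the hood capsule swings mainly on one axis from its anchor.
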PhysicsAsset /BRCosmetics/Characters/Player/Female/Medium/Bodies/F_MED_Neon_Graffiti/Meshes/F_MED_Neon_Graffiti_Dyn_Hood_Physics
bodies {
 dyn_hood_main [{shapeType=Capsule, length=6, radius=6.25, center=<5.0166 -0.6554 0>, rotation=<rot 0 -8.0386 89.9996>}]
 dyn_hood_main_base [{shapeType=Sphere, radius=1}]
 dyn_string_A_base_l [{shapeType=Sphere, radius=1}]
 dyn_string_A_base_r [{shapeType=Sphere, radius=1}]
 dyn_string_A_l [{shapeType=Box, extentX=7, extentY=1.5, extentZ=0.5, center=<3.5514 0 0.0001>}]
 dyn_string_A_r [{shapeType=Box, extentX=7, extentY=1.5, extentZ=0.5, center=<-3.5514 0 -0.0001>, rotation=<rot 0 0 180>}]
 dyn_string_B_l [{shapeType=Box, extentX=6, extentY=1.5, extentZ=0.5, center=<2.7898 0.0417 -0.0344>}]
 dyn_string_B_r [{shapeType=Box, extentX=6, extentY=1.5, extentZ=0.5, center=<-2.7898 -0.0417 0.0344>, rotation=<rot 0 0 180>}]
 dyn_string_C_l [{shapeType=Box, extentX=5.5, extentY=1.5, extentZ=0.5, center=<3.0996 0 0>}]
 dyn_string_C_r [{shapeType=Box, extentX=5.5, extentY=1.5, extentZ=0.5, center=<-3.0996 0 0>, rotation=<rot 0 0 -180>}]
 spine_04 [{shapeType=Capsule, length=9.75, radius=12.25, center=<-2.3383 1.3372 0>, rotation=<rot 0 4.8874 0>}]
 spine_05 [{shapeType=Capsule, length=14, radius=9.75, center=<-1.7891 0.3212 0>}]
}
constraints [
 {
  bone1Name=spine_05
  bone2Name=spine_04
  swing1LimitDegrees=45
  swing1Motion=Free
  swing2LimitDegrees=45
  swing2Motion=Free
  twistLimitDegrees=45
  twistMotion=Free
 }
 {
  bone1Name=dyn_string_A_base_l
  bone2Name=spine_05
  swing1LimitDegrees=45
  swing1Motion=Free
  swing2LimitDegrees=45
  swing2Motion=Free
  twistLimitDegrees=45
  twistMotion=Free
 }
 {
  bone1Name=dyn_string_A_base_r
  bone2Name=spine_05
  swing1LimitDegrees=45
  swing1Motion=Free
  swing2LimitDegrees=45
  swing2Motion=Free
  twistLimitDegrees=45
  twistMotion=Free
 }
 {bone1Name=dyn_string_C_l, bone2Name=dyn_string_B_l, swing1LimitDegrees=15, swing2LimitDegrees=30, twistLimitDegrees=1}
 {bone1Name=dyn_string_B_l, bone2Name=dyn_string_A_l, swing1LimitDegrees=20, swing2LimitDegrees=45, twistLimitDegrees=1}
 {bone1Name=dyn_string_A_l, bone2Name=dyn_string_A_base_l, swing1LimitDegrees=20, swing2LimitDegrees=30, twistLimitDegrees=1}
 {bone1Name=dyn_string_A_r, bone2Name=dyn_string_A_base_r, swing1LimitDegrees=20, swing2LimitDegrees=30, twistLimitDegrees=1}
 {bone1Name=dyn_string_B_r, bone2Name=dyn_string_A_r, swing1LimitDegrees=20, swing2LimitDegrees=45, twistLimitDegrees=1}
 {bone1Name=dyn_string_C_r, bone2Name=dyn_string_B_r, swing1LimitDegrees=15, swing2LimitDegrees=30, twistLimitDegrees=1}
 {
  bone1Name=dyn_hood_main_base
  bone2Name=spine_05
  swing1LimitDegrees=45
  swing1Motion=Free
  swing2LimitDegrees=45
  swing2Motion=Free
  twistLimitDegrees=45
  twistMotion=Free
 }
 {bone1Name=dyn_hood_main, bone2Name=dyn_hood_main_base, swing1LimitDegrees=60, swing2LimitDegrees=5, twistLimitDegrees=5}
]
