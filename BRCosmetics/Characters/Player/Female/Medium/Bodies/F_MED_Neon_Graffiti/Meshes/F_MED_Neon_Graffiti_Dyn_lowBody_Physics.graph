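# Physics asset for the lower-body dynamic bones of the female medium Neon Graffiti cosmetic body, covering two "smoke" attachments on the left thigh.
# A capsule on the left thigh anchors two parallel chains, A and B, each made of a small sphere base body and a capsule body.
# The bases rotate freely on all angular axes against the thigh, while each capsule has narrower swing and twist limit angles relative to its base.
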PhysicsAsset /BRCosmetics/Characters/Player/Female/Medium/Bodies/F_MED_Neon_Graffiti/Meshes/F_MED_Neon_Graffiti_Dyn_lowBody_Physics
bodies {
 dyn_smokeA_base_l [{shapeType=Sphere, radius=1}]
 dyn_smokeA_l [{shapeType=Capsule, length=10.4217, radius=2.9924, center=<-0.4934 1.1543 1.3677>, rotation=<rot 0 -16.6125 -63.9151>}]
 dyn_smokeB_base_l [{shapeType=Sphere, radius=1}]
 dyn_smokeB_l [{shapeType=Capsule, length=10.3921, radius=3.017, center=<-0.6165 0.8406 1.967>, rotation=<rot 0 -16.84 -64.0186>}]
 thigh_l [{shapeType=Capsule, length=21.75, radius=7.75, center=<-15.0454 -0.2878 0.9006>, rotation=<rot 90 -179.1048 -176.9871>}]
}
constraints [
 {bone1Name=dyn_smokeA_l, bone2Name=dyn_smokeA_base_l, swing1LimitDegrees=30, swing2LimitDegrees=30, twistLimitDegrees=30}
 {bone1Name=dyn_smokeB_l, bone2Name=dyn_smokeB_base_l, swing1LimitDegrees=30, swing2LimitDegrees=30, twistLimitDegrees=20}
 {
  bone1Name=dyn_smokeA_base_l
  bone2Name=thigh_l
  swing1LimitDegrees=45
  swing1Motion=Free
  swing2LimitDegrees=45
  swing2Motion=Free
  twistLimitDegrees=45
  twistMotion=Free
 }
 {
  bone1Name=dyn_smokeB_base_l
  bone2Name=thigh_l
  swing1LimitDegrees=45
  swing1Motion=Free
  swing2LimitDegrees=45
  swing2Motion=Free
  twistLimitDegrees=45
  twistMotion=Free
 }
]
