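# Secondary-motion AnimBlueprint for the Neon Graffiti female medium body cosmetic, copying its pose from the attached parent mesh.
# Bone constraints and additive offsets correct thighs, hood, drawstring and smoke bones, with thigh and hood offsets driven by rotations measured in the EventGraph.
# AnimDynamics then simulates right shoestrings and left leg flaps, followed by two RigidBody passes for hood and lower body.
# Two smoke position constraints are disconnected from the output.
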AnimBlueprint /BRCosmetics/Characters/Player/Female/Medium/Bodies/F_MED_Neon_Graffiti/Meshes/F_MED_Neon_Graffiti_AnimBP
target_skeleton=&/F_MED_Neon_Graffiti_Skeleton
graphs [
 AnimGraph {nodes [
   n1 = "Constraint - dyn_smokeA_base_l (PPPP)"!(
    ComponentPose
    Alpha=1
    ConstraintWeights_0=0.4118
    ConstraintWeights_1:float=0.0052
    ConstraintWeights_2:float=0.4969
    ConstraintWeights_3:float=0.0861
   ) -> (Pose) {
    alphaBoolBlend {blendOption=Linear}
    boneToModify {boneName=dyn_smokeA_base_l}
    constraintSetup [
     {targetBone {boneName=pelvis}}
     {targetBone {boneName=thigh_volume_l}}
     {targetBone {boneName=thigh_twist_01_l}}
     {targetBone {boneName=thigh_l}}
    ]
    constraintWeights [1,1,1,1]
   }
   "Constraint - dyn_smokeB_base_l (PPP)"!(ComponentPose=@n1, Alpha=1, ConstraintWeights_0=0.1175, ConstraintWeights_1:float=0.5539, ConstraintWeights_2:float=0.3287) -> (Pose) {
    alphaBoolBlend {blendOption=Linear}
    boneToModify {boneName=dyn_smokeB_base_l}
    constraintSetup [{targetBone {boneName=pelvis}}, {targetBone {boneName=thigh_twist_01_l}}, {targetBone {boneName=thigh_l}}]
    constraintWeights [1,1,1]
   }
   n2 = "Constraint - thigh_volume_l (R)"(ComponentPose=Local To Component(LocalPose=Copy Pose From Mesh() {bUseAttachedParent=true}), Alpha=0.5, ConstraintWeights_0=1) {
    alphaBoolBlend {blendOption=Linear}
    boneToModify {boneName=thigh_volume_l}
    constraintSetup [{transformType=Rotation, targetBone {boneName=thigh_l}}]
    constraintWeights [1]
   }
   n3 = "Constraint - thigh_volume_r (R)"(ComponentPose=@n2, Alpha=0.5, ConstraintWeights_0=1) {
    alphaBoolBlend {blendOption=Linear}
    boneToModify {boneName=thigh_volume_r}
    constraintSetup [{transformType=Rotation, targetBone {boneName=thigh_r}}]
    constraintWeights [1]
   }
   n4 = "Transform (Modify) Bone - Bone: thigh_volume_l"(ComponentPose=@n3, Alpha=Get thigh_rot_l(), Translation="-1,-2,1", Rotation="0,0,0", Scale="1,1,1") {
    translationMode=BMM_Additive
    translationSpace=BCS_BoneSpace
    alphaBoolBlend {blendOption=Linear}
    boneToModify {boneName=thigh_volume_l}
    translation=<-1 -2 1>
   }
   n5 = "Transform (Modify) Bone - Bone: thigh_volume_r"(ComponentPose=@n4, Alpha=Get thigh_rot_r(), Translation="1,2,-1", Rotation="0,0,0", Scale="1,1,1") {
    translationMode=BMM_Additive
    translationSpace=BCS_BoneSpace
    alphaBoolBlend {blendOption=Linear}
    boneToModify {boneName=thigh_volume_r}
    translation=<1 2 -1>
   }
   n6 = "Constraint - dyn_string_A_base_l (PPPP)"!(
    ComponentPose=@n5
    Alpha=1
    ConstraintWeights_0=0.0188
    ConstraintWeights_1:float=0.0162
    ConstraintWeights_2:float=0.9506
    ConstraintWeights_3:float=0.0145
   ) -> (Pose) {
    alphaBoolBlend {blendOption=Linear}
    boneToModify {boneName=dyn_string_A_base_l}
    constraintSetup [{targetBone {boneName=dyn_hood_l}}, {targetBone {boneName=spine_04}}, {targetBone {boneName=spine_05}}, {targetBone {boneName=clavicle_l}}]
    constraintWeights [1,1,1,1]
   }
   n7 = "Constraint - dyn_string_A_base_r (PPPP)"!(
    ComponentPose=@n6
    Alpha=1
    ConstraintWeights_0=0.0445
    ConstraintWeights_1:float=0.0201
    ConstraintWeights_2:float=0.8868
    ConstraintWeights_3:float=0.0486
   ) -> (Pose) {
    alphaBoolBlend {blendOption=Linear}
    boneToModify {boneName=dyn_string_A_base_r}
    constraintSetup [{targetBone {boneName=clavicle_r}}, {targetBone {boneName=spine_04}}, {targetBone {boneName=spine_05}}, {targetBone {boneName=dyn_hood_r}}]
    constraintWeights [1,1,1,1]
   }
   n8 = "Transform (Modify) Bone - Bone: dyn_string_A_base_l"(ComponentPose=@n7, Alpha=1, Translation="0,0,0", Rotation="1,0,0", Scale="1,1,1") {
    rotationMode=BMM_Additive
    rotationSpace=BCS_BoneSpace
    translationMode=BMM_Additive
    translationSpace=BCS_BoneSpace
    alphaBoolBlend {blendOption=Linear}
    boneToModify {boneName=dyn_string_A_base_l}
    rotation {pitch=1}
   }
   n9 = "Transform (Modify) Bone - Bone: dyn_string_A_base_r"(ComponentPose=@n8, Alpha=1, Translation="0,0,0", Rotation="1,0,0", Scale="1,1,1") {
    rotationMode=BMM_Additive
    rotationSpace=BCS_BoneSpace
    translationMode=BMM_Additive
    translationSpace=BCS_BoneSpace
    alphaBoolBlend {blendOption=Linear}
    boneToModify {boneName=dyn_string_A_base_r}
    rotation {pitch=1}
   }
   n10 = "Constraint - dyn_smokeA_base_l (R)"(ComponentPose=@n9, Alpha=1, ConstraintWeights_0=1) {
    alphaBoolBlend {blendOption=Linear}
    boneToModify {boneName=dyn_smokeA_base_l}
    constraintSetup [{transformType=Rotation, targetBone {boneName=thigh_twist_01_l}}]
    constraintWeights [1]
   }
   n11 = "Constraint - dyn_smokeB_base_l (R)"(ComponentPose=@n10, Alpha=1, ConstraintWeights_0=1) {
    alphaBoolBlend {blendOption=Linear}
    boneToModify {boneName=dyn_smokeB_base_l}
    constraintSetup [{transformType=Rotation, targetBone {boneName=thigh_twist_01_l}}]
    constraintWeights [1]
   }
   n12 = Get headFwdBk() -> (headFwdBk:double)
   n13 = "Transform (Modify) Bone - Bone: dyn_hood_base_l"(ComponentPose=@n11, Alpha=@n12, Translation="-1,0,0", Rotation="0,10,0", Scale="1,1,1") {
    rotationMode=BMM_Additive
    rotationSpace=BCS_BoneSpace
    translationMode=BMM_Additive
    translationSpace=BCS_BoneSpace
    alphaBoolBlend {blendOption=Linear}
    alphaScaleBiasClamp {bMapRange=true, inRange {max=-30, min=-50}, outRange {max=0, min=1}}
    boneToModify {boneName=dyn_hood_base_l}
    rotation {yaw=10}
    translation {x=-1}
   }
   n14 = "Transform (Modify) Bone - Bone: dyn_hood_back_base"(ComponentPose=@n13, Alpha=@n12, Translation="-2.5,0,0", Rotation="0,30,0", Scale="1,1,1") {
    rotationMode=BMM_Additive
    rotationSpace=BCS_BoneSpace
    translationMode=BMM_Additive
    translationSpace=BCS_BoneSpace
    alphaBoolBlend {blendOption=Linear}
    alphaScaleBiasClamp {bMapRange=true, inRange {max=-30, min=-50}, outRange {max=0, min=1}}
    boneToModify {boneName=dyn_hood_back_base}
    rotation {yaw=30}
    translation {x=-2.5}
   }
   n15 = "Transform (Modify) Bone - Bone: dyn_hood_base_r"(ComponentPose=@n14, Alpha=@n12, Translation="-1,0,0", Rotation="0,10,0", Scale="1,1,1") {
    rotationMode=BMM_Additive
    rotationSpace=BCS_BoneSpace
    translationMode=BMM_Additive
    translationSpace=BCS_BoneSpace
    alphaBoolBlend {blendOption=Linear}
    alphaScaleBiasClamp {bMapRange=true, inRange {max=-30, min=-50}, outRange {max=0, min=1}}
    boneToModify {boneName=dyn_hood_base_r}
    rotation {yaw=10}
    translation {x=-1}
   }
   n16 = "Anim Dynamics - Chain: dyn_shoestring_a_1_r -> dyn_shoestring_a_2_r"(ComponentPose=@n15, Alpha=1, ExternalForce="0,0,0") {
    angularDampingOverride=0.995
    angularSpringConstant=300
    bAngularSpring=true
    bChain=true
    bGravityOverrideInSimSpace=true
    bOverrideAngularDamping=true
    bOverrideLinearDamping=true
    gravityScale=0.35
    linearDampingOverride=0.995
    simulationSpace=RootRelative
    alphaBoolBlend {blendOption=Linear}
    boundBone {boneName=dyn_shoestring_a_1_r}
    chainEnd {boneName=dyn_shoestring_a_2_r}
    componentAppliedLinearAccClamp=<3000 3000 3000>
    componentLinearAccScale=<0.2 0.2 0.2>
    componentLinearVelScale=<0.2 0.2 0.2>
    physicsBodyDefinitions [{
      sphereCollisionRadius=0
      boundBone {boneName=dyn_shoestring_a_1_r}
      boxExtents=<5 10 10>
      constraintSetup {angularTargetAxis=AxisZ, angularLimitsMax=<0 1 40>, angularLimitsMin=<0 -1 0>, angularTarget=<0 0 0>}
      localJointOffset=<2 0 0>
     }]
   }
   n17 = "Anim Dynamics - Chain: dyn_shoestring_b_1_r -> dyn_shoestring_b_2_r"(ComponentPose=@n16, Alpha=1, ExternalForce="0,0,0") {
    angularDampingOverride=0.995
    angularSpringConstant=300
    bAngularSpring=true
    bChain=true
    bGravityOverrideInSimSpace=true
    bOverrideAngularDamping=true
    bOverrideLinearDamping=true
    gravityScale=0.35
    linearDampingOverride=0.995
    simulationSpace=RootRelative
    alphaBoolBlend {blendOption=Linear}
    boundBone {boneName=dyn_shoestring_b_1_r}
    chainEnd {boneName=dyn_shoestring_b_2_r}
    componentAppliedLinearAccClamp=<1000 1000 1000>
    componentLinearAccScale=<0.1 0.1 0.1>
    componentLinearVelScale=<0.1 0.1 0.1>
    physicsBodyDefinitions [{
      sphereCollisionRadius=0
      boundBone {boneName=dyn_shoestring_b_1_r}
      boxExtents=<5 10 10>
      constraintSetup {angularTargetAxis=AxisZ, angularLimitsMax=<0 5 30>, angularLimitsMin=<0 -5 -5>, angularTarget=<0 0 0>}
      localJointOffset=<3 0 0>
     }]
   }
   n18 = Anim Dynamics - Bone: dyn_loleg_flap_a_l(ComponentPose=@n17, Alpha=1, ExternalForce="0,0,0") {
    angularDampingOverride=1
    angularSpringConstant=300
    bAngularSpring=true
    bGravityOverrideInSimSpace=true
    bOverrideAngularDamping=true
    bOverrideLinearDamping=true
    linearDampingOverride=0.9999
    simulationSpace=RootRelative
    alphaBoolBlend {blendOption=Linear}
    boundBone {boneName=dyn_loleg_flap_a_l}
    componentAppliedLinearAccClamp=<1000 1000 1000>
    componentLinearAccScale=<0.3 0.3 0.3>
    componentLinearVelScale=<0.3 0.3 0.3>
    physicsBodyDefinitions [{
      sphereCollisionRadius=0
      boundBone {boneName=dyn_loleg_flap_a_l}
      boxExtents=<5 10 10>
      constraintSetup {twistAxis=AxisZ, angularLimitsMax=<5 10 0>, angularLimitsMin=<-15 -10 0>, angularTarget=<0 0 0>}
      localJointOffset=<3 0 0>
     }]
   }
   n19 = Anim Dynamics - Bone: dyn_loleg_flap_b_l(ComponentPose=@n18, Alpha=1, ExternalForce="0,0,0") {
    angularDampingOverride=1
    angularSpringConstant=300
    bAngularSpring=true
    bGravityOverrideInSimSpace=true
    bOverrideAngularDamping=true
    bOverrideLinearDamping=true
    linearDampingOverride=0.9999
    simulationSpace=RootRelative
    alphaBoolBlend {blendOption=Linear}
    boundBone {boneName=dyn_loleg_flap_b_l}
    componentAppliedLinearAccClamp=<1000 1000 1000>
    componentLinearAccScale=<0.3 0.3 0.3>
    componentLinearVelScale=<0.3 0.3 0.3>
    physicsBodyDefinitions [{
      sphereCollisionRadius=0
      boundBone {boneName=dyn_loleg_flap_b_l}
      boxExtents=<5 10 10>
      constraintSetup {twistAxis=AxisZ, angularLimitsMax=<5 10 0>, angularLimitsMin=<-15 -10 0>, angularTarget=<0 0 0>}
      localJointOffset=<4 0 0>
     }]
   }
   n20 = RigidBody(ComponentPose=@n19) {
    bClampLinearTranslationLimitToRefPose=true
    bOverrideWorldGravity=true
    overridePhysicsAsset=&/F_MED_Neon_Graffiti_Dyn_Hood_Physics
    simulationSpace=BaseBoneSpace
    alphaBoolBlend {blendOption=Linear}
    baseBoneRef {boneName=spine_01}
    componentLinearAccScale=<0.3 0.3 0.5>
    componentLinearVelScale=<0.3 0.3 0.5>
    overrideWorldGravity {z=-980}
   }
   n21 = RigidBody(ComponentPose=@n20) {
    bClampLinearTranslationLimitToRefPose=true
    bOverrideWorldGravity=true
    overridePhysicsAsset=&/F_MED_Neon_Graffiti_Dyn_lowBody_Physics
    simulationSpace=BaseBoneSpace
    alphaBoolBlend {blendOption=Linear}
    baseBoneRef {boneName=pelvis}
    componentLinearAccScale=<0.5 0.5 0.4>
    componentLinearVelScale=<0.5 0.5 0.4>
    overrideWorldGravity {z=-980}
   }
   Output Pose(Result=Component To Local(ComponentPose=@n21)) {layerGroup=None}
  ]}
 EventGraph {nodes [
   Event BlueprintUpdateAnimation()
   > n1 = Sequence!(execute) -> (then_0, then_1, then_2:Exec)
   n2 = GetOwnerSkeletalMeshComponent()
   n3 = GetDeltaTransformFromRefPose(self=@n2, BoneName=head, BaseName=spine_05)
   > Set headFwdBk(headFwdBk:double=BreakRotator(InRot=BreakTransform(InTransform=@n3).Rotation).Yaw) -> (Output_Get:double)
   n4 = BreakTransform(InTransform=GetDeltaTransformFromRefPose(self=@n2, BoneName=thigh_r, BaseName=pelvis))
   Set thigh_rot_r(execute=@n1.then_2, thigh_rot_r=MapRangeClamped(Value=BreakRotator(InRot=@n4.Rotation).Yaw, InRangeA=0, InRangeB=-105))
   n5 = BreakTransform(InTransform=GetDeltaTransformFromRefPose(self=@n2, BoneName=thigh_l, BaseName=pelvis))
   Set thigh_rot_l(execute=@n1.then_1, thigh_rot_l=MapRangeClamped(Value=BreakRotator(InRot=@n5.Rotation).Yaw, InRangeA=-20, InRangeB=-90))
   TryGetPawnOwner()
  ]}
]
variables [
 headFwdBk {type=number}
 thigh_rot_l {type=number}
 thigh_rot_r {type=number}
]
