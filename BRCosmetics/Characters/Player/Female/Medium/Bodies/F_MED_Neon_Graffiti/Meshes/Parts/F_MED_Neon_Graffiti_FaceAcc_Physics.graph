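# Physics asset for the F_MED_Neon_Graffiti face accessory mesh of the female medium player cosmetic, defining physics bodies for its dynamic hair strands.
# A sphere on the head anchors six capsule chains, hair A, B and C on each side, three bones long except the four-bone right A chain.
# Each capsule is constrained to its parent, with swing limits loosening toward the tips and twist kept uniformly tight.
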PhysicsAsset /BRCosmetics/Characters/Player/Female/Medium/Bodies/F_MED_Neon_Graffiti/Meshes/Parts/F_MED_Neon_Graffiti_FaceAcc_Physics
bodies {
 dyn_hairA_A_l [{shapeType=Capsule, length=2.75, radius=1.5, center=<2.2484 0 0>, rotation=<rot 90 0 0>}]
 dyn_hairA_A_r [{shapeType=Capsule, length=2.142, radius=1.7462, center=<2.4059 0.0001 0>, rotation=<rot 90 0 0>}]
 dyn_hairA_B_l [{shapeType=Capsule, length=2.5, radius=2, center=<2.8464 0.0001 -0.0123>, rotation=<rot 86.8991 0 0>}]
 dyn_hairA_B_r [{shapeType=Capsule, length=2.4104, radius=1.6989, center=<2.7081 0 0>, rotation=<rot 74.7924 0 0>}]
 dyn_hairA_C_l [{shapeType=Capsule, length=3.5, radius=1.75, center=<3.0026 0 0>, rotation=<rot 83.2233 0 0>}]
 dyn_hairA_C_r [{shapeType=Capsule, length=0.3267, radius=2.1869, center=<2.2344 0.0001 0>, rotation=<rot 90 0 0>}]
 dyn_hairA_D_r [{shapeType=Capsule, length=3.4457, radius=2.1098, center=<3.192 0 0>, rotation=<rot 90 0 0>}]
 dyn_hairB_A_l [{shapeType=Capsule, length=2.25, radius=1.75, center=<2.0086 -0.0001 -0.383>, rotation=<rot 69.5932 0 0.0001>}]
 dyn_hairB_A_r [{shapeType=Capsule, length=1.4754, radius=1.4637, center=<1.5448 0.0001 0>, rotation=<rot 90 0 0>}]
 dyn_hairB_B_l [{shapeType=Capsule, length=0.25, radius=2, center=<1.6686 0 0>, rotation=<rot 90 0 0>}]
 dyn_hairB_B_r [{shapeType=Capsule, length=1.6991, radius=1.8481, center=<1.7705 0 -0.3524>, rotation=<rot 78.4194 0.0001 0.0001>}]
 dyn_hairB_C_l [{shapeType=Capsule, length=3, radius=2.25, center=<2.8714 0.0001 -0.3314>, rotation=<rot 85.8731 -0.0002 -0.0002>}]
 dyn_hairB_C_r [{shapeType=Capsule, length=2.4186, radius=2.3284, center=<2.6341 0 0>, rotation=<rot 90 0 0>}]
 dyn_hairC_A_l [{shapeType=Capsule, length=3.25, radius=2, center=<2.5168 -0.2831 -0.156>, rotation=<rot 78.5891 0 0>}]
 dyn_hairC_A_r [{shapeType=Capsule, length=3.4436, radius=2.1128, center=<3.4471 1.0317 -0.6005>, rotation=<rot 80.1175 0 -12.4215>}]
 dyn_hairC_B_l [{shapeType=Capsule, length=1.5, radius=2.25, center=<2.9328 0 0>, rotation=<rot 90 0 0>}]
 dyn_hairC_B_r [{shapeType=Capsule, length=1.5806, radius=2.3804, center=<2.5809 0 0>, rotation=<rot 90 0 0>}]
 dyn_hairC_C_l [{shapeType=Capsule, length=2.75, radius=2.25, center=<3.2662 0 0>, rotation=<rot 90 0 0>}]
 dyn_hairC_C_r [{shapeType=Capsule, length=3.0158, radius=2.5617, center=<3.4901 0 -0.2244>, rotation=<rot 86.541 0 0>}]
 head [{shapeType=Sphere, radius=9, center=<7.0299 -0.3742 -0.0001>}]
}
constraints [
 {bone1Name=dyn_hairA_A_l, bone2Name=head, swing1LimitDegrees=15, swing2LimitDegrees=15, twistLimitDegrees=5}
 {bone1Name=dyn_hairA_B_l, bone2Name=dyn_hairA_A_l, swing1LimitDegrees=20, swing2LimitDegrees=20, twistLimitDegrees=5}
 {bone1Name=dyn_hairA_C_l, bone2Name=dyn_hairA_B_l, swing1LimitDegrees=25, swing2LimitDegrees=25, twistLimitDegrees=5}
 {bone1Name=dyn_hairB_A_l, bone2Name=head, swing1LimitDegrees=15, swing2LimitDegrees=15, twistLimitDegrees=5}
 {bone1Name=dyn_hairB_B_l, bone2Name=dyn_hairB_A_l, swing1LimitDegrees=20, swing2LimitDegrees=20, twistLimitDegrees=5}
 {bone1Name=dyn_hairB_C_l, bone2Name=dyn_hairB_B_l, swing1LimitDegrees=25, swing2LimitDegrees=25, twistLimitDegrees=5}
 {bone1Name=dyn_hairC_A_l, bone2Name=head, swing1LimitDegrees=15, swing2LimitDegrees=15, twistLimitDegrees=5}
 {bone1Name=dyn_hairC_B_l, bone2Name=dyn_hairC_A_l, swing1LimitDegrees=20, swing2LimitDegrees=20, twistLimitDegrees=5}
 {bone1Name=dyn_hairC_C_l, bone2Name=dyn_hairC_B_l, swing1LimitDegrees=25, swing2LimitDegrees=25, twistLimitDegrees=5}
 {bone1Name=dyn_hairA_A_r, bone2Name=head, swing1LimitDegrees=15, swing2LimitDegrees=15, twistLimitDegrees=5}
 {bone1Name=dyn_hairA_B_r, bone2Name=dyn_hairA_A_r, swing1LimitDegrees=20, swing2LimitDegrees=20, twistLimitDegrees=5}
 {bone1Name=dyn_hairA_C_r, bone2Name=dyn_hairA_B_r, swing1LimitDegrees=25, swing2LimitDegrees=25, twistLimitDegrees=5}
 {bone1Name=dyn_hairA_D_r, bone2Name=dyn_hairA_C_r, swing1LimitDegrees=30, swing2LimitDegrees=30, twistLimitDegrees=5}
 {bone1Name=dyn_hairC_A_r, bone2Name=head, swing1LimitDegrees=15, swing2LimitDegrees=15, twistLimitDegrees=5}
 {bone1Name=dyn_hairC_B_r, bone2Name=dyn_hairC_A_r, swing1LimitDegrees=20, swing2LimitDegrees=20, twistLimitDegrees=5}
 {bone1Name=dyn_hairC_C_r, bone2Name=dyn_hairC_B_r, swing1LimitDegrees=25, swing2LimitDegrees=25, twistLimitDegrees=5}
 {bone1Name=dyn_hairB_A_r, bone2Name=head, swing1LimitDegrees=15, swing2LimitDegrees=15, twistLimitDegrees=5}
 {bone1Name=dyn_hairB_B_r, bone2Name=dyn_hairB_A_r, swing1LimitDegrees=20, swing2LimitDegrees=20, twistLimitDegrees=5}
 {bone1Name=dyn_hairB_C_r, bone2Name=dyn_hairB_B_r, swing1LimitDegrees=25, swing2LimitDegrees=25, twistLimitDegrees=5}
]
