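# Post-pose Anim Blueprint for the Neon Graffiti female character's face accessory mesh, adding secondary physics motion.
# It copies the attached parent mesh's pose, then runs a RigidBody simulation in root-bone space with scaled component motion and overridden gravity.
# The clamped current LOD drives the alpha of a slight additive translation on faceAttach, so the offset applies only below LOD 0 detail.
# The event graph holds no logic.
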AnimBlueprint /BRCosmetics/Characters/Player/Female/Medium/Bodies/F_MED_Neon_Graffiti/Meshes/Parts/F_MED_Neon_Graffiti_FaceAcc_AnimBP
target_skeleton=&/F_MED_Neon_Graffiti_FaceAcc_Skeleton
graphs [
 AnimGraph {nodes [
   n1 = RigidBody(ComponentPose=Local To Component(LocalPose=Copy Pose From Mesh() {bUseAttachedParent=true})) {
    bClampLinearTranslationLimitToRefPose=true
    bOverrideWorldGravity=true
    overridePhysicsAsset=&/F_MED_Neon_Graffiti_FaceAcc_Physics
    simulationSpace=BaseBoneSpace
    alphaBoolBlend {blendOption=Linear}
    baseBoneRef {boneName=root}
    componentLinearAccScale=<0.4 0.4 0.55>
    componentLinearVelScale=<0.4 0.4 0.55>
    overrideWorldGravity {z=-980}
   }
   n2 = Get CurrentLOD()
   n3 = "Transform (Modify) Bone - Bone: faceAttach"(
    ComponentPose=@n1
    Alpha="Clamp (Float)"(Value="To Float (Integer)"(InInt=@n2), Min=0, Max=1)
    Translation="0,0.3,0"
    Rotation="0,0,0"
    Scale="1,1,1"
   ) {translationMode=BMM_Additive, alphaBoolBlend {blendOption=Linear}, boneToModify {boneName=faceAttach}, translation {y=0.3}}
   Output Pose(Result=Component To Local(ComponentPose=@n3)) {layerGroup=None}
  ]}
 EventGraph {nodes [
   TryGetPawnOwner()
   Event BlueprintUpdateAnimation()
  ]}
]
variables []
